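# Control Rig for the M_MED_Scuffle male body cosmetic, adding corrective deformation and clothing adjustments.
# A sequence runs the shared DeformRig function, then splits movement inputs into directional variables.
# Skirt bases blend between pelvis and thighs, with thigh-raise and frontend offsets; Kalman-smoothed movement, scaled by pelvis height, shifts the front jacket.
# Foot yaw weights rear foot support offsets, while pelvis height sets Crouching and glute correctives.
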ControlRigBlueprint /BRCosmetics/Characters/Player/Male/Medium/Bodies/M_MED_Scuffle/Meshes/M_MED_Scuffle_CtrlRig
graphs [
 RigVMFunctionLibrary {links [], nodes []}
 RigVMFunction_Sequence_1_ContainedGraph {nodes [
   n1 = RigVMFunctionEntryNode()
   RigVMFunction_Sequence = &RigVMFunction_Sequence(ExecuteContext=@n1.ExecuteContext)
   n2 = &RigVMFunction_Sequence(ExecuteContext=@RigVMFunction_Sequence.B)
   n3 = &RigVMFunction_Sequence(ExecuteContext=@n2.B)
   RigVMFunctionReturnNode(A=@RigVMFunction_Sequence.A, B=@n2.A, C=@n3.A, D=@n3.B)
  ]}
 RigVMModel {nodes [
   RigUnit_BeginExecution = &RigUnit_BeginExecution()
   RigVMFunction_Sequence = RigVMAggregateNode(ExecuteContext=@RigUnit_BeginExecution.ExecutePin)
   FN_DeformRig_FNC = RigVMFunctionReferenceNode(
    DeformRig_ Curve="(deformRig_l_upperarm_up_alpha_72_#1=False,deformRig_r_upperarm_up_alpha_73_#2=False,deformRig_l_upperarm_fr_alpha_74_#3=False,deformRig_r_upperarm_fr_alpha_75_CB51B8 ...(truncated)"
    DeformRig_Bool="(upperarm_1_#4=True,elbow_3_#5=True,wrist_7_#6=True,groin_8_#7=True,glute_17_#8=True,knee_9_40C408324CA261 ...(truncated)"
    DeformRig_Debug=(upperarm_debug_18_#4=False,wrist_debug_20_#6=False,lat_debug_26_#9=False,groinGlute_debug_27_#7=False)
    DeformRig_Value="(upperarm_up_18_#10=(Rotation=(X=0,Y=0,Z=0,W=1),Translation=(X=0,Y=0,Z=8),Scale3D=(X=0.7,Y=1,Z=1)),upperarm_fwd_fr_19_#11=(Rotation=(X=0,Y=0,Z=0,W=1),Translation=(X=0,Y=10,Z=0),Scale3D=(X=0.5,Y=1,Z= ...(truncated)"
    ExecuteContext=@RigVMFunction_Sequence.A
   )
   n1 = &RigVMFunction_Sequence(ExecuteContext=@FN_DeformRig_FNC.ExecuteContext)
   n2 = RigVMVariableNode() {variable=Forward Backward Movement}
   Remap = &RigVMFunction_MathDoubleRemap(SourceMaximum=1, SourceMinimum=0, TargetMaximum=1, TargetMinimum=0, Value=@n2.Value)
   RigVMVariableNode(Value=@Remap.Result, ExecuteContext=@n1.A) {variable=Forward Movement}
   > RigVMVariableNode(Value=&RigVMFunction_MathDoubleRemap(SourceMaximum=0, SourceMinimum=-1, TargetMaximum=0, TargetMinimum=1, Value=@n2.Value).Result) {variable=Backward Movement}
   n3 = RigVMVariableNode() {variable=Left Right Movement}
   RigVMVariableNode(
    Value=&RigVMFunction_MathDoubleRemap(SourceMaximum=0, SourceMinimum=-1, TargetMaximum=0, TargetMinimum=1, Value=@n3.Value).Result
    ExecuteContext=@n1.B
   ) {variable="Left  Movement"}
   > RigVMVariableNode(Value=&RigVMFunction_MathDoubleRemap(SourceMaximum=1, SourceMinimum=0, TargetMaximum=1, TargetMinimum=0, Value=@n3.Value).Result) {variable="Right  Movement"}
   Get Delta Ref = RigVMFunctionReferenceNode(Item=(Type=Bone,Name="foot_r"), ExecuteContext=@RigVMFunction_Sequence.C)
   > &RigUnit_ModifyTransforms(
    ItemToModify=((Item=(Type=Bone,Name="support_foot_bk_r"),Transform=(Rotation=(X=0,Y=0,Z=0,W=1),Translation=(X=-10.7043,Y=9.0064,Z=0),Scale3D=(X=1,Y=1,Z=1))))
    WeightMinimum=0
    Weight=&RigVMFunction_MathDoubleRemap(SourceMaximum=50, SourceMinimum=25, TargetMaximum=1, TargetMinimum=0, Value=@"Get Delta Ref".Result.Rotation.Yaw).Result
   )
   > Get Delta Ref_1 = RigVMFunctionReferenceNode(Item=(Type=Bone,Name="foot_l"))
   > &RigUnit_ModifyTransforms(
    ItemToModify=((Item=(Type=Bone,Name="support_foot_bk_l"),Transform=(Rotation=(X=0,Y=0,Z=0,W=1),Translation=(X=-8.2473,Y=-12.3037,Z=0),Scale3D=(X=1,Y=1,Z=1))))
    WeightMinimum=0
    Weight=&RigVMFunction_MathDoubleRemap(SourceMaximum=50, SourceMinimum=25, TargetMaximum=1, TargetMinimum=0, Value=@"Get Delta Ref_1".Result.Rotation.Yaw).Result
   )
   n4 = &RigVMFunction_MathDoubleRemap(
    SourceMaximum=45
    SourceMinimum=65
    TargetMaximum=1
    TargetMinimum=0
    Value=&RigVMFunction_MathVectorDistance(
     A=(X=0,Y=0,Z=0)
     B=(X=0,Y=0,Z=0)
     B.Z=&RigUnit_GetTransform(Item=(Type=Bone,Name="root"), bInitial=False).Transform.Translation.Z
     A.Z=&RigUnit_GetTransform(Item=(Type=Bone,Name="pelvis"), bInitial=False).Transform.Translation.Z
    ).Result
   )
   RigVMVariableNode(Value=@n4.Result, ExecuteContext=@RigVMFunction_Sequence.D) {variable=Crouching}
   AlphaInterp = &RigVMFunction_AlphaInterp(InterpSpeedDecreasing=10, InterpSpeedIncreasing=10, Scale=1, bClampResult=False, bInterpResult=False, Value=@n4.Result)
   > &RigUnit_ModifyTransforms(
    ItemToModify="((Item=(Type=Bone,Name=\"deform_glute_l\"),Transform=(Rotation=(X=0,Y=0,Z=0,W=1),Translation=(X=-5.1018,Y=0.6741,Z=3.4676),Scale3D=(X=1,Y=1,Z=1))),(Item=(Type=Bone,Name=\"deform_glute_r\"),Transform=(Rotation=(X=0,Y=0,Z=0,W=1),Translation=(X=0,Y=0,Z=0),Scale3D ...(truncated)"
    WeightMinimum=0
    Weight=@AlphaInterp.Result
   )
   n5 = &RigVMFunction_KalmanFloat(BufferSize=16, Value=1) {template="KalmanFilter::Execute(in Value,in BufferSize,out Result)"}
   &RigUnit_ParentConstraint(
    AdvancedSettings=(InterpolationType=Average,RotationOrderForFilter=XZY)
    Child=(Type=Bone,Name="dyn_skirt_l_base")
    Parents=((Item=(Type=Bone,Name="pelvis"),Weight=0.2),(Item=(Type=Bone,Name="thigh_l"),Weight=0.8))
    Weight=1
    Parents.1.Weight=@n5.Result
    ExecutePin=@RigVMFunction_Sequence.B
   )
   > &RigUnit_ParentConstraint(
    AdvancedSettings=(InterpolationType=Average,RotationOrderForFilter=XZY)
    Child=(Type=Bone,Name="dyn_skirt_r_base")
    Parents=((Item=(Type=Bone,Name="pelvis"),Weight=0.2),(Item=(Type=Bone,Name="thigh_r"),Weight=0.8))
    Weight=1
    Parents.1.Weight=@n5.Result
   )
   DISPATCH_RigDispatch_GetAnimAttribute = RigVMDispatchNode(BoneName=Root, Default=0, Name=deformRig_l_thigh_up_alpha) {template="DISPATCH_RigDispatch_GetAnimAttribute(in Name,in BoneName,in Default,out Value,out Found)"}
   n6 = &RigVMFunction_AlphaInterp(
    InterpSpeedDecreasing=10
    InterpSpeedIncreasing=10
    Scale=1
    bClampResult=False
    bInterpResult=False
    Value=@DISPATCH_RigDispatch_GetAnimAttribute.Value
   )
   > &RigUnit_ModifyTransforms(
    ItemToModify=((Item=(Type=Bone,Name="dyn_skirt_l_base"),Transform=(Rotation=(X=-0.3682,Y=0.0219,Z=-0.0551,W=0.9278),Translation=(X=1.29,Y=0,Z=1.6442),Scale3D=(X=1,Y=1,Z=1))))
    WeightMinimum=0
    Weight=@n6.Result
   )
   n7 = &RigVMFunction_AlphaInterp(
    InterpSpeedDecreasing=10
    InterpSpeedIncreasing=10
    Scale=1
    bClampResult=False
    bInterpResult=False
    Value=RigVMDispatchNode(BoneName=Root, Default=0, Name=deformRig_r_thigh_up_alpha) {template="DISPATCH_RigDispatch_GetAnimAttribute(in Name,in BoneName,in Default,out Value,out Found)"}.Value
   )
   > &RigUnit_ModifyTransforms(
    ItemToModify=((Item=(Type=Bone,Name="dyn_skirt_r_base"),Transform=(Rotation=(X=0.0989,Y=0,Z=0,W=0.9951),Translation=(X=-2.2469,Y=-0.0001,Z=1.0934),Scale3D=(X=1,Y=1,Z=1))))
    WeightMinimum=0
    Weight=@n7.Result
   )
   RigVMFunction_MathBoolToFloat = &RigVMFunction_MathBoolToFloat(Value=RigVMVariableNode() {variable=Is In Frontend}.Value)
   n8 = &RigVMFunction_AlphaInterp(
    InterpSpeedDecreasing=10
    InterpSpeedIncreasing=10
    Scale=1
    bClampResult=False
    bInterpResult=False
    Value=@RigVMFunction_MathBoolToFloat.Result
   )
   > &RigUnit_ModifyTransforms(
    ItemToModify="((Item=(Type=Bone,Name=\"dyn_skirt_l_base\"),Transform=(Rotation=(X=-0.061,Y=0,Z=0,W=0.9981),Translation=(X=0,Y=0,Z=0),Scale3D=(X=1,Y=1,Z=1))),(Item=(Type=Bone,Name=\"dyn_skirt_r_base\"),Transform=(Rotation=(X=0.0436,Y=0,Z=0,W=0.999),Translation=(X=0,Y=0,Z=0), ...(truncated)"
    WeightMinimum=0
    Weight=@n8.Result
   )
   Add = &RigVMFunction_MathDoubleAdd(A=RigVMVariableNode() {variable=Left Right Movement}.Value, B=RigVMVariableNode() {variable=Forward Backward Movement}.Value)
   Absolute = &RigVMFunction_MathDoubleAbs(Value=@Add.Result)
   KalmanFilter = &RigVMFunction_KalmanFloat(
    BufferSize=16
    Value=&RigVMFunction_MathDoubleRemap(SourceMaximum=1, SourceMinimum=0, TargetMaximum=1, TargetMinimum=0, Value=@Absolute.Result).Result
   ) {template="KalmanFilter::Execute(in Value,in BufferSize,out Result)"}
   n9 = &RigUnit_GetTransform(Item=(Type=Bone,Name="pelvis"), bInitial=False)
   n10 = &RigUnit_GetTransform(Item=(Type=Bone,Name="root"), bInitial=False)
   Multiply = &RigVMFunction_MathDoubleMul(
    B=@KalmanFilter.Result
    A=&RigVMFunction_MathDoubleRemap(
     SourceMaximum=75
     SourceMinimum=90
     TargetMaximum=1
     TargetMinimum=0
     Value=&RigVMFunction_MathVectorDistance(A=(X=0,Y=0,Z=0), B=(X=0,Y=0,Z=0), A.Z=@n9.Transform.Translation.Z, B.Z=@n10.Transform.Translation.Z).Result
    ).Result
   )
   n11 = &RigVMFunction_AlphaInterp(InterpSpeedDecreasing=24, InterpSpeedIncreasing=24, Scale=1, bClampResult=False, bInterpResult=False, Value=@Multiply.Result)
   > &RigUnit_ModifyTransforms(
    ItemToModify=((Item=(Type=Bone,Name="dyn_jacket_frnt_base"),Transform=(Rotation=(X=0,Y=0,Z=0,W=1),Translation=(X=-8,Y=0,Z=0),Scale3D=(X=1,Y=1,Z=1))))
    WeightMinimum=0
    Weight=@n11.Result
   )
   > RigUnit_SetRelativeTransformForItem = /Script/ControlRig.RigUnit_SetRelativeTransformForItem(
    Child=(Type=Bone,Name="dyn_jacket_frnt_02")
    Parent=(Type=Bone,Name="dyn_jacket_frnt_02")
    Value=(Rotation=(X=0,Y=-0.0872,Z=0,W=0.9962),Translation=(X=-10,Y=0,Z=0),Scale3D=(X=1,Y=1,Z=1))
    bParentInitial=False
    bPropagateToChildren=false
    Weight=@n11.Result
   ) {
    is=struct RigVMUnitNode
    template="Set Relative Transform::Execute(in Child,in Parent,in bParentInitial,in Value,in Weight,in bPropagateToChildren)"
   }
  ]}
]
hierarchy [
 root {loc=<0 0 0>, rot=<0 0 0 1>}
 pelvis {parent=root, loc=<0 -0.7071 93.1191>, rot=<0.0127 0.707 -0.0127 -0.707>}
 spine_01 {parent=pelvis, loc=<6.7494 -0.6388 0>, rot=<0 0 0.034 -0.9994>}
 spine_02 {parent=spine_01, loc=<7.2518 -0.035 0>, rot=<0 0 0.0238 -0.9997>}
 spine_03 {parent=spine_02, loc=<6.0104 0 0>, rot=<0 0 0.0149 -0.9999>}
 spine_04 {parent=spine_03, loc=<8.1024 0 0>, rot=<0 0 -0.0132 -0.9999>}
 spine_05 {parent=spine_04, loc=<10.0146 0 0>, rot=<0 0 -0.0015 -1>}
 neck_01 {parent=spine_05, loc=<11.3138 0.1529 0>, rot=<0 0 -0.2003 -0.9797>}
 neck_02 {parent=neck_01, loc=<4.5627 0 0>, rot=<0 0 0.0504 -0.9987>}
 head {parent=neck_02, loc=<5.3914 0 0>, rot=<0 0 0.115 -0.9934>}
 dyn_hood_strap_master {parent=head, loc=<-8.2778 9.1358 0.1746>, rot=<0 0 0.8783 0.4781>}
 dyn_hood_strap_r_base {parent=dyn_hood_strap_master, loc=<2.0198 0.2275 0.3519>, rot=<-0.6515 -0.2244 -0.3019 -0.6588>}
 dyn_hood_strap_r_01 {parent=dyn_hood_strap_r_base, loc=<0 0 0>, rot=<0 0 0 1>}
 dyn_hood_strap_r_02 {parent=dyn_hood_strap_r_01, loc=<7.0573 -0.2859 -0.8559>, rot=<0.0093 -0.07 0.0136 -0.9974>}
 dyn_hood_strap_r_03 {parent=dyn_hood_strap_r_02, loc=<8.3095 -0.0109 -0.0244>, rot=<0.0373 -0.018 0.0044 -0.9991>}
 dyn_hood_strap_l_base {parent=dyn_hood_strap_master, loc=<1.9123 0.0814 -0.5452>, rot=<-0.6399 -0.3427 -0.2066 -0.6561>}
 dyn_hood_strap_l_01 {parent=dyn_hood_strap_l_base, loc=<0 0 0>, rot=<0 0 0 1>}
 dyn_hood_strap_l_02 {parent=dyn_hood_strap_l_01, loc=<7.6873 0.3652 -0.513>, rot=<-0.0421 -0.0951 -0.0472 -0.9935>}
 dyn_hood_strap_l_03 {parent=dyn_hood_strap_l_02, loc=<8.4231 0.222 0.0519>, rot=<0.0016 0.0218 -0.015 -0.9996>}
 clavicle_l {parent=spine_05, loc=<4.3575 -0.1466 -2.3223>, rot=<0.0465 -0.7289 -0.009 -0.6829>}
 upperarm_l {parent=clavicle_l, loc=<15.5111 0 0>, rot=<0.0226 -0.3234 -0.0088 -0.9459>}
 lowerarm_l {parent=upperarm_l, loc=<26.6514 0 0>, rot=<0.0391 -0.0099 -0.2453 -0.9686>}
 hand_l {parent=lowerarm_l, loc=<22.3491 0 0>, rot=<-0.7081 -0.0507 0.0975 -0.6975>}
 thumb_01_l {parent=hand_l, loc=<1.3118 -0.3191 -2.459>, rot=<0.6546 -0.4042 0.0664 -0.6354>}
 thumb_02_l {parent=thumb_01_l, loc=<5.9336 -0.0448 0.0401>, rot=<-0.0037 0.0213 0.298 -0.9543>}
 thumb_03_l {parent=thumb_02_l, loc=<2.2223 0.3682 -0.0907>, rot=<0.0118 -0.013 0.0066 -0.9998>}
 index_metacarpal_l {parent=hand_l, loc=<1.6535 0.0183 -2.96>, rot=<0.0158 -0.0124 -0.0173 -0.9996>}
 index_01_l {parent=index_metacarpal_l, loc=<8.3008 0.0357 -0.0524>, rot=<-0.0044 0.0485 0.0924 -0.9945>}
 index_02_l {parent=index_01_l, loc=<3.3587 -0.3004 0.0452>, rot=<-0.002 0.0027 0.3225 -0.9466>}
 index_03_l {parent=index_02_l, loc=<2.5578 -0.123 -0.0252>, rot=<-0.002 0.0067 0.2558 -0.9667>}
 middle_metacarpal_l {parent=hand_l, loc=<2.1856 -0.1725 -0.5984>, rot=<0.006 0.0168 -0.0329 -0.9993>}
 middle_01_l {parent=middle_metacarpal_l, loc=<7.3275 -0.2447 0.0129>, rot=<-0.0208 0.0042 0.0858 -0.9961>}
 middle_02_l {parent=middle_01_l, loc=<4.0123 0.1401 0.1175>, rot=<-0.0218 0.0258 0.3512 -0.9357>}
 middle_03_l {parent=middle_02_l, loc=<2.6077 0.0534 0.1373>, rot=<0.0082 0.0188 0.1708 -0.9851>}
 ring_metacarpal_l {parent=hand_l, loc=<2.2074 -0.3625 1.3638>, rot=<-0.0284 0.0722 -0.0458 -0.9959>}
 ring_01_l {parent=ring_metacarpal_l, loc=<7.0295 -0.6197 -0.1446>, rot=<-0.0489 -0.0252 0.1892 -0.9804>}
 ring_02_l {parent=ring_01_l, loc=<3.6174 0.4647 -0.1027>, rot=<0 0 0.2806 -0.9598>}
 ring_03_l {parent=ring_02_l, loc=<2.4152 0.114 0.0199>, rot=<0 0 0.1189 -0.9929>}
 pinky_metacarpal_l {parent=hand_l, loc=<1.7616 -0.9753 3.3814>, rot=<-0.0397 0.0814 -0.069 -0.9935>}
 pinky_01_l {parent=pinky_metacarpal_l, loc=<7.1709 -0.4831 -0.1094>, rot=<-0.0654 -0.0035 0.2242 -0.9723>}
 pinky_02_l {parent=pinky_01_l, loc=<2.8274 0.0893 -0.1043>, rot=<0.0037 0.0011 0.2854 -0.9584>}
 pinky_03_l {parent=pinky_02_l, loc=<2.2046 0.0307 -0.0078>, rot=<-0.0002 0.0098 0.1445 -0.9895>}
 deform_wrist_dn_base_l {parent=hand_l, loc=<0 -2.2342 0>, rot=<0 0 0 -1>}
 deform_wrist_dn_l {parent=deform_wrist_dn_base_l, loc=<0 0 0>, rot=<0 0 0 1>}
 deform_wrist_up_base_l {parent=hand_l, loc=<0 2.2682 0>, rot=<0 0 0 -1>}
 deform_wrist_up_l {parent=deform_wrist_up_base_l, loc=<0 0 0>, rot=<0 0 0 1>}
 lowerarm_twist_02_l {parent=lowerarm_l, loc=<9.6199 0 0>, rot=<0 0 0 -1>}
 lowerarm_twist_01_l {parent=lowerarm_l, loc=<17.8549 0 0>, rot=<0 0 0 -1>}
 deform_elbow_in_base_l {parent=lowerarm_l, loc=<0.2112 5.259 0.8274>, rot=<0 0 0 -1>}
 deform_elbow_in_l {parent=deform_elbow_in_base_l, loc=<0 0 0>, rot=<0 0 0 1>}
 deform_elbow_out_base_l {parent=lowerarm_l, loc=<0 -4 0>, rot=<0 0 0 -1>}
 deform_elbow_out_l {parent=deform_elbow_out_base_l, loc=<0 0 0>, rot=<0 0 0 1>}
 upperarm_twist_01_l {parent=upperarm_l, loc=<0.0041 0 0>, rot=<0 0 0 -1>}
 deform_upperarm_bk_base_l {parent=upperarm_twist_01_l, loc=<1.0813 -7.2193 -1.9453>, rot=<0 0 0 -1>}
 deform_upperarm_bk_l {parent=deform_upperarm_bk_base_l, loc=<0 0 0>, rot=<0 0 0 1>}
 deform_upperarm_fr_base_l {parent=upperarm_twist_01_l, loc=<-0.3691 7.0506 -2.765>, rot=<0 0 0 -1>}
 deform_upperarm_fr_l {parent=deform_upperarm_fr_base_l, loc=<0 0 0>, rot=<0 0 0 1>}
 deform_upperarm_up_base_l {parent=upperarm_twist_01_l, loc=<-3.2216 -1.1391 3.3311>, rot=<0 0 0 -1>}
 deform_upperarm_up_l {parent=deform_upperarm_up_base_l, loc=<0 0 0>, rot=<0 0 0 1>}
 upperarm_twist_02_l {parent=upperarm_l, loc=<13.7941 0 0>, rot=<0 0 0 -1>}
 clavicle_r {parent=spine_05, loc=<4.3575 -0.1466 2.3223>, rot=<0.7289 0.0465 -0.6829 0.009>}
 upperarm_r {parent=clavicle_r, loc=<-15.5111 0 0>, rot=<0.0226 -0.3234 -0.0088 -0.9459>}
 lowerarm_r {parent=upperarm_r, loc=<-26.6514 0 0>, rot=<0.0391 -0.0099 -0.2453 -0.9686>}
 hand_r {parent=lowerarm_r, loc=<-22.3491 0 0>, rot=<-0.7081 -0.0507 0.0975 -0.6975>}
 thumb_01_r {parent=hand_r, loc=<-1.3118 0.3191 2.459>, rot=<0.6546 -0.4042 0.0664 -0.6354>}
 thumb_02_r {parent=thumb_01_r, loc=<-5.9336 0.0448 -0.0401>, rot=<-0.0037 0.0213 0.298 -0.9543>}
 thumb_03_r {parent=thumb_02_r, loc=<-2.2223 -0.3682 0.0907>, rot=<0.0118 -0.013 0.0066 -0.9998>}
 index_metacarpal_r {parent=hand_r, loc=<-1.6535 -0.0183 2.96>, rot=<0.0158 -0.0124 -0.0173 -0.9996>}
 index_01_r {parent=index_metacarpal_r, loc=<-8.3008 -0.0357 0.0524>, rot=<-0.0044 0.0485 0.0924 -0.9945>}
 index_02_r {parent=index_01_r, loc=<-3.3587 0.3004 -0.0452>, rot=<-0.002 0.0027 0.3225 -0.9466>}
 index_03_r {parent=index_02_r, loc=<-2.5578 0.123 0.0252>, rot=<-0.002 0.0067 0.2558 -0.9667>}
 middle_metacarpal_r {parent=hand_r, loc=<-2.1856 0.1725 0.5984>, rot=<0.006 0.0168 -0.0329 -0.9993>}
 middle_01_r {parent=middle_metacarpal_r, loc=<-7.3275 0.2447 -0.0129>, rot=<-0.0208 0.0042 0.0858 -0.9961>}
 middle_02_r {parent=middle_01_r, loc=<-4.0123 -0.1401 -0.1175>, rot=<-0.0218 0.0258 0.3512 -0.9357>}
 middle_03_r {parent=middle_02_r, loc=<-2.6077 -0.0534 -0.1373>, rot=<0.0082 0.0188 0.1708 -0.9851>}
 ring_metacarpal_r {parent=hand_r, loc=<-2.2074 0.3625 -1.3638>, rot=<-0.0284 0.0722 -0.0458 -0.9959>}
 ring_01_r {parent=ring_metacarpal_r, loc=<-7.0295 0.6197 0.1446>, rot=<-0.0489 -0.0252 0.1892 -0.9804>}
 ring_02_r {parent=ring_01_r, loc=<-3.6174 -0.4647 0.1027>, rot=<0 0 0.2806 -0.9598>}
 ring_03_r {parent=ring_02_r, loc=<-2.4152 -0.114 -0.0199>, rot=<0 0 0.1189 -0.9929>}
 pinky_metacarpal_r {parent=hand_r, loc=<-1.7616 0.9753 -3.3814>, rot=<-0.0397 0.0814 -0.069 -0.9935>}
 pinky_01_r {parent=pinky_metacarpal_r, loc=<-7.1709 0.4831 0.1094>, rot=<-0.0654 -0.0035 0.2242 -0.9723>}
 pinky_02_r {parent=pinky_01_r, loc=<-2.8274 -0.0893 0.1043>, rot=<0.0037 0.0011 0.2854 -0.9584>}
 pinky_03_r {parent=pinky_02_r, loc=<-2.2046 -0.0307 0.0078>, rot=<-0.0002 0.0098 0.1445 -0.9895>}
 deform_wrist_dn_base_r {parent=hand_r, loc=<-0.0004 2.2345 -0.0001>, rot=<0 0 0 -1>}
 deform_wrist_dn_r {parent=deform_wrist_dn_base_r, loc=<0 0 0>, rot=<0 0 0 1>}
 deform_wrist_up_base_r {parent=hand_r, loc=<0.0003 -2.2684 0.0001>, rot=<0 0 0 -1>}
 deform_wrist_up_r {parent=deform_wrist_up_base_r, loc=<0 0 0>, rot=<0 0 0 1>}
 lowerarm_twist_02_r {parent=lowerarm_r, loc=<-9.6199 0 0>, rot=<0 0 0 -1>}
 lowerarm_twist_01_r {parent=lowerarm_r, loc=<-17.8549 0 0>, rot=<0 0 0 -1>}
 deform_elbow_in_base_r {parent=lowerarm_r, loc=<-0.2111 -5.2591 -0.8275>, rot=<0 0 0 -1>}
 deform_elbow_in_r {parent=deform_elbow_in_base_r, loc=<-0.0001 0 0.0002>, rot=<0 0 0 1>}
 deform_elbow_out_base_r {parent=lowerarm_r, loc=<0 4 -0.0001>, rot=<0 0 0 -1>}
 deform_elbow_out_r {parent=deform_elbow_out_base_r, loc=<0 0 0>, rot=<0 0 0 1>}
 upperarm_twist_01_r {parent=upperarm_r, loc=<-0.0041 0 0>, rot=<0 0 0 -1>}
 deform_upperarm_bk_base_r {parent=upperarm_twist_01_r, loc=<-1.0814 7.2192 1.9455>, rot=<0 0 0 1>}
 deform_upperarm_bk_r {parent=deform_upperarm_bk_base_r, loc=<0 0 0>, rot=<0 0 0 1>}
 deform_upperarm_fr_base_r {parent=upperarm_twist_01_r, loc=<0.3689 -7.0506 2.7652>, rot=<0 0 0 1>}
 deform_upperarm_fr_r {parent=deform_upperarm_fr_base_r, loc=<0 0 0>, rot=<0 0 0 1>}
 deform_upperarm_up_base_r {parent=upperarm_twist_01_r, loc=<3.2214 1.1391 -3.3308>, rot=<0 0 0 1>}
 deform_upperarm_up_r {parent=deform_upperarm_up_base_r, loc=<0 0 0>, rot=<0 0 0 -1>}
 upperarm_twist_02_r {parent=upperarm_r, loc=<-13.7941 0 0>, rot=<0 0 0 -1>}
 attach_backpack {parent=spine_05, loc=<-0.123 -13.4727 0>, rot=<0 -0.7071 0 -0.7071>}
 attach_cape {parent=spine_05, loc=<9.017 -11.714 0>, rot=<0.0284 -0.7065 -0.0284 -0.7065>}
 dyn_jacket_strap_bk_base {parent=spine_05, loc=<4.6048 -11.9557 -0.6749>, rot=<-0.6971 0.1655 0.1667 -0.6774>}
 dyn_jacket_strap_bk_01 {parent=dyn_jacket_strap_bk_base, loc=<0 0 0>, rot=<0 0 0 1>}
 deform_pec_aim_base_l {parent=spine_04, loc=<7.8732 12.6388 -2.4123>, rot=<0.0102 -0.4602 0.3979 -0.7936>}
 deform_pec_aim_l {parent=deform_pec_aim_base_l, loc=<0 0 0>, rot=<0 0 0 1>}
 deform_pec_base_l {parent=deform_pec_aim_l, loc=<10.766 0 0>, rot=<0 0 0 1>}
 deform_pec_l {parent=deform_pec_base_l, loc=<0.055 0 0>, rot=<0 0 0 1>}
 deform_pec_aim_base_r {parent=spine_04, loc=<7.8727 12.6388 2.4123>, rot=<-0.4602 -0.0102 0.7936 0.3979>}
 deform_pec_aim_r {parent=deform_pec_aim_base_r, loc=<0 0 0>, rot=<0 0 0 1>}
 deform_pec_base_r {parent=deform_pec_aim_r, loc=<-10.7662 -0.0006 -0.0007>, rot=<0 0 0 1>}
 deform_pec_r {parent=deform_pec_base_r, loc=<0 0 0>, rot=<0 0 0 1>}
 support_jacket_l {parent=spine_04, loc=<-1.3308 -1.1418 -14.5099>, rot=<0.0238 -0.7661 -0.0335 -0.6414>}
 support_jacket_r {parent=spine_04, loc=<-1.3307 -1.1418 14.5099>, rot=<-0.0238 0.7661 -0.0335 -0.6414>}
 dyn_jacket_strap_r_base {parent=spine_03, loc=<-2.8811 -3.6262 12.8079>, rot=<-0.2295 0.37 -0.7469 -0.5025>}
 dyn_jacket_strap_r_01 {parent=dyn_jacket_strap_r_base, loc=<0 0 0>, rot=<0 0 0 1>}
 dyn_jacket_strap_l_base {parent=spine_03, loc=<-2.8807 -3.6261 -12.8079>, rot=<-0.5025 -0.7469 0.37 -0.2295>}
 dyn_jacket_strap_l_01 {parent=dyn_jacket_strap_l_base, loc=<0 0 0>, rot=<0 0 0 1>}
 dyn_jacket_frnt_base {parent=spine_03, loc=<-0.1047 11.2434 0>, rot=<-0.0387 0.706 0.706 -0.0387>}
 dyn_jacket_frnt_01 {parent=dyn_jacket_frnt_base, loc=<0 0 0>, rot=<0 0 0 1>}
 dyn_jacket_frnt_02 {parent=dyn_jacket_frnt_01, loc=<8.4987 0 0.5563>, rot=<0 0.0362 0 -0.9993>}
 dyn_jacket_frnt_03 {parent=dyn_jacket_frnt_02, loc=<8.3219 0 0.0717>, rot=<0 0.0159 0 -0.9999>}
 dyn_jacket_frnt_scrap_l {parent=dyn_jacket_frnt_03, loc=<8.3244 -8.4414 -0.8524>, rot=<-0.0808 0 0 -0.9967>}
 dyn_jacket_frnt_scrap_r {parent=dyn_jacket_frnt_03, loc=<8.4798 5.0478 -0.1345>, rot=<0.0083 0.0804 0.9915 -0.1019>}
 dyn_hoody_frnt_r_base {parent=spine_03, loc=<-0.7248 11.2927 6.0446>, rot=<0.0071 -0.7071 -0.0071 -0.7071>}
 dyn_hoody_frnt_r_01 {parent=dyn_hoody_frnt_r_base, loc=<-0.0053 0 0>, rot=<0 0 0 1>}
 dyn_hoody_frnt_r_02 {parent=dyn_hoody_frnt_r_01, loc=<0 0 -7.9265>, rot=<0 -0.7071 0 -0.7071>}
 dyn_hoody_frnt_r_03 {parent=dyn_hoody_frnt_r_02, loc=<7.4122 0 0>, rot=<0 0 0 1>}
 dyn_hoody_scrap_r_base {parent=dyn_hoody_frnt_r_03, loc=<9.0937 0.2626 1.0779>, rot=<0.7628 -0.056 -0.0299 -0.6435>}
 dyn_hoody_scrap_r_01 {parent=dyn_hoody_scrap_r_base, loc=<0 0 0>, rot=<0 0 0 1>}
 dyn_hoody_frnt_l_base {parent=spine_03, loc=<-0.7248 11.2927 -6.2364>, rot=<0.0071 -0.7071 -0.0071 -0.7071>}
 dyn_hoody_frnt_l_01 {parent=dyn_hoody_frnt_l_base, loc=<0 0 0>, rot=<0 0 0 1>}
 dyn_hoody_frnt_l_02 {parent=dyn_hoody_frnt_l_01, loc=<0 0 -7.9265>, rot=<0 -0.7071 0 -0.7071>}
 dyn_hoody_frnt_l_03 {parent=dyn_hoody_frnt_l_02, loc=<7.4122 0 0>, rot=<0 0 0 1>}
 dyn_hoody_scrap_l_base {parent=dyn_hoody_frnt_l_03, loc=<8.7909 -0.7852 2.5338>, rot=<0.631 0.0357 0.0588 -0.7728>}
 dyn_hoody_scrap_l_01 {parent=dyn_hoody_scrap_l_base, loc=<0 0 0>, rot=<0 0 0 1>}
 thigh_l {parent=pelvis, loc=<-2.5693 -0.0109 -7.7501>, rot=<0.0012 0.0572 0.0155 -0.9982>}
 calf_l {parent=thigh_l, loc=<-39.0484 0 0>, rot=<0.0006 -0.0161 -0.0391 -0.9991>}
 foot_l {parent=calf_l, loc=<-42.561 0.1077 0.4292>, rot=<0.0418 -0.0455 0.0343 -0.9975>}
 ball_l {parent=foot_l, loc=<-6.1599 13.8264 0.6806>, rot=<0.0001 -0.0001 -0.7063 -0.7079>}
 support_foot_bk_l {parent=foot_l, loc=<0 -4.059 0>, rot=<0 0 0 1>}
 calf_twist_02_l {parent=calf_l, loc=<-18.8064 -0.0967 0.0397>, rot=<0.0237 0.0095 -0.0035 -0.9997>}
 dyn_lace_l_base {parent=calf_twist_02_l, loc=<-8.5448 3.5692 -0.2318>, rot=<-0.5019 -0.4604 -0.4911 -0.5431>}
 dyn_lace_l_02 {parent=dyn_lace_l_base, loc=<-0.0002 0.3623 0.0871>, rot=<-0.0427 0.1455 -0.6886 -0.7091>}
 dyn_lace_l_01 {parent=dyn_lace_l_base, loc=<-0.0068 -0.3109 0.0917>, rot=<0.0913 -0.097 -0.7285 -0.672>}
 dyn_lace_l_03 {parent=dyn_lace_l_base, loc=<0.3916 -0.7644 -0.6843>, rot=<0.0829 -0.6946 0.0811 -0.71>}
 dyn_lace_l_05 {parent=dyn_lace_l_03, loc=<4.7323 0.6276 -0.2036>, rot=<-0.001 -0.0146 -0.1272 -0.9918>}
 dyn_lace_l_04 {parent=dyn_lace_l_base, loc=<0.3957 0.3837 -1.1768>, rot=<-0.0747 -0.6696 -0.0737 -0.7353>}
 dyn_lace_l_06 {parent=dyn_lace_l_04, loc=<3.2594 0.1619 -0.163>, rot=<0.0017 -0.0377 -0.0445 -0.9983>}
 calf_twist_01_l {parent=calf_l, loc=<-35.813 -0.0967 0.0397>, rot=<0.0237 0.0095 -0.0035 -0.9997>}
 deform_knee_in_base_l {parent=calf_l, loc=<1.298 -5.2957 -0.0496>, rot=<0 0 0 -1>}
 deform_knee_in_l {parent=deform_knee_in_base_l, loc=<0 0 0>, rot=<0 0 0 1>}
 deform_knee_out_base_l {parent=calf_l, loc=<0 4 0>, rot=<0 0 0 -1>}
 deform_knee_out_l {parent=deform_knee_out_base_l, loc=<0 0 0>, rot=<0 0 0 1>}
 dyn_legs_scrap_l_01 {parent=calf_l, loc=<-8.5754 -5.0543 -4.7842>, rot=<0.0725 -0.9132 0.3476 -0.1998>}
 dyn_legs_scrap_l_02 {parent=calf_l, loc=<-9.5334 1.1763 -4.7527>, rot=<0.7751 0.129 -0.5842 -0.2031>}
 dyn_legs_scrap_l_03 {parent=calf_l, loc=<-9.3412 3.5051 1.1445>, rot=<0.1234 0.5942 0.791 0.0781>}
 dyn_legs_scrap_l_04 {parent=calf_l, loc=<-9.0648 -8.2198 -0.2366>, rot=<0.0717 -0.731 0.6718 -0.0958>}
 dyn_legs_scrap_l_05 {parent=calf_l, loc=<-8.976 -3.8639 4.7462>, rot=<-0.2481 -0.1575 -0.0387 -0.9551>}
 thigh_twist_01_l {parent=thigh_l, loc=<-21.6303 0 0>, rot=<0.0013 0 0 -1>}
 dyn_legs_knife_l_base {parent=thigh_twist_01_l, loc=<-5.0471 0.374 -6.4803>, rot=<0.9968 0.0113 -0.079 -0.0026>}
 dyn_legs_knife_l_01 {parent=dyn_legs_knife_l_base, loc=<0 0 0>, rot=<0 0 0 1>}
 deform_groin_base_l {parent=thigh_l, loc=<-2.1908 8.1142 -1.7442>, rot=<0 0 0 -1>}
 deform_groin_l {parent=deform_groin_base_l, loc=<0 0 0>, rot=<0 0 0 1>}
 dyn_skirt_l_base {parent=thigh_l, loc=<-6.6721 9.945 -0.5678>, rot=<-0.0136 -0.7769 -0.6291 -0.021>}
 dyn_skirt_l_01 {parent=dyn_skirt_l_base, loc=<0 0 0>, rot=<0 0 0 1>}
 dyn_skirt_l_knife_base {parent=dyn_skirt_l_01, loc=<4.2249 -2.5746 -0.3127>, rot=<-0.0081 -0.0484 0.9545 -0.2942>}
 dyn_skirt_l_knife_01 {parent=dyn_skirt_l_knife_base, loc=<0 0 0>, rot=<0 0 0 1>}
 thigh_r {parent=pelvis, loc=<-2.5693 -0.0109 7.7501>, rot=<0.0572 -0.0012 0.9982 0.0155>}
 calf_r {parent=thigh_r, loc=<39.0484 0 0>, rot=<0.0006 -0.0161 -0.0391 -0.9991>}
 foot_r {parent=calf_r, loc=<42.561 -0.1077 -0.4292>, rot=<0.0418 -0.0455 0.0343 -0.9975>}
 ball_r {parent=foot_r, loc=<6.1599 -13.8264 -0.6806>, rot=<0.0001 -0.0001 -0.7063 -0.7079>}
 support_foot_bk_r {parent=foot_r, loc=<0 4.059 0>, rot=<0 -1 0 0>}
 calf_twist_02_r {parent=calf_r, loc=<18.8064 0.0967 -0.0397>, rot=<0.0237 0.0095 -0.0035 -0.9997>}
 dyn_lace_r_base {parent=calf_twist_02_r, loc=<8.795 -3.5929 0.2215>, rot=<0.5284 -0.5767 -0.4559 0.4247>}
 dyn_lace_r_01 {parent=dyn_lace_r_base, loc=<0.0316 0.9419 0.3803>, rot=<-0.0826 0.094 -0.6739 -0.7281>}
 dyn_lace_r_02 {parent=dyn_lace_r_base, loc=<0.3593 1.2237 -0.3904>, rot=<-0.1784 -0.5009 -0.3599 -0.7667>}
 dyn_lace_r_04 {parent=dyn_lace_r_base, loc=<0.4306 -0.7378 -0.6604>, rot=<-0.0122 -0.5312 0.28 -0.7996>}
 dyn_lace_r_06 {parent=dyn_lace_r_04, loc=<3.2326 0.0588 -0.9178>, rot=<0.145 -0.1969 -0.0436 -0.9687>}
 dyn_lace_r_03 {parent=dyn_lace_r_base, loc=<0.7358 0.7093 -0.6192>, rot=<0.2023 -0.4799 -0.2547 -0.8148>}
 dyn_lace_r_05 {parent=dyn_lace_r_03, loc=<3.3026 -0.1776 -1.5347>, rot=<-0.0636 -0.3384 0.0717 -0.9361>}
 calf_twist_01_r {parent=calf_r, loc=<35.813 0.0967 -0.0397>, rot=<0.0237 0.0095 -0.0035 -0.9997>}
 deform_knee_in_base_r {parent=calf_r, loc=<-1.298 5.2957 0.0496>, rot=<0 0 0 -1>}
 deform_knee_in_r {parent=deform_knee_in_base_r, loc=<0 0 0>, rot=<0 0 0 1>}
 deform_knee_out_base_r {parent=calf_r, loc=<0 -4 0>, rot=<0 0 0 -1>}
 deform_knee_out_r {parent=deform_knee_out_base_r, loc=<0 0 0>, rot=<0 0 0 1>}
 dyn_legs_scrap_r_03 {parent=calf_r, loc=<9.5485 -2.8069 -2.5248>, rot=<0.8819 -0.0225 -0.0548 0.4677>}
 dyn_legs_scrap_r_05 {parent=calf_r, loc=<8.976 3.8639 -4.7462>, rot=<0.9551 0.0387 -0.1575 -0.2481>}
 dyn_legs_scrap_r_02 {parent=calf_r, loc=<10.4974 -3.3822 2.5047>, rot=<-0.5326 0.1348 0.0698 -0.8326>}
 dyn_legs_scrap_r_06 {parent=calf_r, loc=<11.9359 0.8675 5.448>, rot=<-0.0451 0.0875 -0.0126 -0.9951>}
 dyn_legs_scrap_r_01 {parent=calf_r, loc=<10.5065 6.1133 4.4534>, rot=<0.3834 0.1419 -0.082 -0.9089>}
 dyn_legs_scrap_r_04 {parent=calf_r, loc=<9.2255 7.8309 -1.2578>, rot=<0.779 0.0703 -0.0844 -0.6173>}
 thigh_twist_01_r {parent=thigh_r, loc=<21.6303 0 0>, rot=<0.0013 0 0 -1>}
 dyn_legs_knife_r_base {parent=thigh_twist_01_r, loc=<5.0471 -0.3739 6.4803>, rot=<0.079 -0.0026 0.9968 -0.0113>}
 dyn_legs_knife_r_01 {parent=dyn_legs_knife_r_base, loc=<0 0 0>, rot=<0 0 0 -1>}
 deform_groin_base_r {parent=thigh_r, loc=<2.1908 -8.1142 1.7442>, rot=<0 0 0 -1>}
 deform_groin_r {parent=deform_groin_base_r, loc=<0 0 0>, rot=<0 0 0 1>}
 dyn_skirt_r_base {parent=thigh_r, loc=<6.6647 -10.0399 0.2765>, rot=<-0.6417 0.0164 0.0166 -0.7666>}
 dyn_skirt_r_01 {parent=dyn_skirt_r_base, loc=<0 0 0>, rot=<0 0 0 1>}
 dyn_skirt_r_knife_base {parent=dyn_skirt_r_01, loc=<4.2006 2.9146 -0.4256>, rot=<-0.0147 0.0633 0.9552 0.2887>}
 dyn_skirt_r_knife_01 {parent=dyn_skirt_r_knife_base, loc=<0 0 0>, rot=<0 0 0 1>}
 deform_glute_base_l {parent=pelvis, loc=<-8.8064 -8.8222 -8.4287>, rot=<0.0275 0.0501 -0.4472 -0.8926>}
 deform_glute_l {parent=deform_glute_base_l, loc=<0 0 0>, rot=<0 0 0 1>}
 deform_glute_base_r {parent=pelvis, loc=<-8.8062 -8.8222 8.4287>, rot=<0.0501 -0.0275 0.8926 -0.4472>}
 deform_glute_r {parent=deform_glute_base_r, loc=<0 0 0>, rot=<0 0 0 1>}
 dyn_skirt_bk_mid_base {parent=pelvis, loc=<4.6242 -10.7188 0>, rot=<0.1318 -0.9913 0 0>}
 dyn_skirt_bk_mid_01 {parent=dyn_skirt_bk_mid_base, loc=<0 0 0>, rot=<0 0 0 1>}
 dyn_jacket_knife_bk_base {parent=dyn_skirt_bk_mid_01, loc=<-1.1612 -0.3049 -5.9852>, rot=<-0.0643 -0.7422 -0.0584 -0.6645>}
 dyn_jacket_knife_bk_01 {parent=dyn_jacket_knife_bk_base, loc=<0 0 0>, rot=<0 0 0 1>}
]
variables [
 Backward Movement {type=double, default=0}
 Crouching {type=double, default=0}
 Forward Backward Movement {type=double, default=0}
 Forward Movement {type=double, default=0}
 Is In Frontend {type=bool, default=False}
 "Left  Movement" {type=double, default=0}
 Left Right Movement {type=double, default=0}
 "Right  Movement" {type=double, default=0}
]
