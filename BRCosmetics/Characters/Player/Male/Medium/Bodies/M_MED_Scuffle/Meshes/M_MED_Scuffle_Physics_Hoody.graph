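# Physics asset for the M_MED_Scuffle male medium character body, covering the hoody's dynamic front jacket bones.
# Pelvis, spine and thigh capsules form a body proxy with free angular constraints, and the jacket base hangs off spine_03.
# A jacket chain runs from that base through three segments to left and right scraps, with angular limits that favor one swing axis.
# The third segment adds a wide box shape alongside a zero-size capsule.
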PhysicsAsset /BRCosmetics/Characters/Player/Male/Medium/Bodies/M_MED_Scuffle/Meshes/M_MED_Scuffle_Physics_Hoody
bodies {
 dyn_jacket_frnt_01 [{shapeType=Capsule, length=48.0155, radius=3.6677, shapeName=dyn_jacket_frnt_01_capsule, center=<4.4772 0 1.5184>, rotation=<rot 0 0 90>}]
 dyn_jacket_frnt_02 [{shapeType=Capsule, length=77.2733, radius=4.2335, shapeName=dyn_jacket_frnt_02_capsule, center=<4.4128 0 1.3363>, rotation=<rot 0 0 90>}]
 dyn_jacket_frnt_03 [
  {shapeType=Capsule, length=0, radius=0, shapeName=dyn_jacket_frnt_03_capsule, center=<5.4678 0 0>, rotation=<rot 0 0 90>}
  {shapeType=Box, extentX=17.5595, extentY=94.3618, extentZ=6.203, shapeName=dyn_jacket_frnt_03_box, center=<6.5184 0 0>}
 ]
 dyn_jacket_frnt_base [{shapeType=Capsule, length=0.505, radius=0.505, shapeName=dyn_jacket_frnt_base_capsule, rotation=<rot 0 -90 0>}]
 dyn_jacket_frnt_scrap_l [{shapeType=Capsule, length=5.7359, radius=3.1632, shapeName=dyn_jacket_frnt_scrap_l_capsule, center=<2.9323 0 1.5273>, rotation=<rot 0 0 90>}]
 dyn_jacket_frnt_scrap_r [{
   shapeType=Capsule
   length=11.6326
   radius=3.3674
   shapeName=dyn_jacket_frnt_scrap_r_capsule
   center=<-2.5614 0 1.4542>
   rotation=<rot 0 0 90>
  }]
 pelvis [{shapeType=Capsule, length=13.025, radius=10.3226, shapeName=pelvis_capsule, center=<-6.6284 -0.5161 0>}]
 spine_01 [{shapeType=Capsule, length=14.5816, radius=7.7821, shapeName=spine_01_capsule, center=<-2.5341 0 0>}]
 spine_02 [{shapeType=Capsule, length=14.5816, radius=7.7821, shapeName=spine_02_capsule, center=<-2.9832 0 0>}]
 spine_03 [{shapeType=Capsule, length=22.2709, radius=8.2503, shapeName=spine_03_capsule, center=<-2.0115 0 0>}]
 thigh_l [{shapeType=Capsule, length=54.8745, radius=10.0312, shapeName=thigh_l_capsule, center=<-33.8475 1.3573 0>, rotation=<rot 90 0 0>}]
 thigh_r [{shapeType=Capsule, length=54.8745, radius=10.0312, shapeName=thigh_l_capsule, center=<33.8475 -1.8531 0>, rotation=<rot 90 0 180>}]
}
constraints [
 {
  bone1Name=spine_01
  bone2Name=pelvis
  swing1LimitDegrees=45
  swing1Motion=Free
  swing2LimitDegrees=45
  swing2Motion=Free
  twistLimitDegrees=45
  twistMotion=Free
 }
 {
  bone1Name=spine_02
  bone2Name=spine_01
  swing1LimitDegrees=45
  swing1Motion=Free
  swing2LimitDegrees=45
  swing2Motion=Free
  twistLimitDegrees=45
  twistMotion=Free
 }
 {
  bone1Name=spine_03
  bone2Name=spine_02
  swing1LimitDegrees=45
  swing1Motion=Free
  swing2LimitDegrees=45
  swing2Motion=Free
  twistLimitDegrees=45
  twistMotion=Free
 }
 {
  bone1Name=thigh_l
  bone2Name=pelvis
  swing1LimitDegrees=45
  swing1Motion=Free
  swing2LimitDegrees=45
  swing2Motion=Free
  twistLimitDegrees=45
  twistMotion=Free
 }
 {
  bone1Name=dyn_jacket_frnt_base
  bone2Name=spine_03
  swing1LimitDegrees=45
  swing1Motion=Free
  swing2LimitDegrees=45
  swing2Motion=Free
  twistLimitDegrees=45
  twistMotion=Free
 }
 {bone1Name=dyn_jacket_frnt_01, bone2Name=dyn_jacket_frnt_base, swing1LimitDegrees=2, swing2LimitDegrees=70, twistLimitDegrees=0}
 {bone1Name=dyn_jacket_frnt_02, bone2Name=dyn_jacket_frnt_01, swing1LimitDegrees=2, swing2LimitDegrees=40, twistLimitDegrees=40}
 {bone1Name=dyn_jacket_frnt_03, bone2Name=dyn_jacket_frnt_02, swing1LimitDegrees=2, swing2LimitDegrees=40, twistLimitDegrees=40}
 {bone1Name=dyn_jacket_frnt_scrap_l, bone2Name=dyn_jacket_frnt_03, swing1LimitDegrees=20, swing2LimitDegrees=20, twistLimitDegrees=10}
 {bone1Name=dyn_jacket_frnt_scrap_r, bone2Name=dyn_jacket_frnt_03, swing1LimitDegrees=20, swing2LimitDegrees=20, twistLimitDegrees=10}
 {
  bone1Name=thigh_r
  bone2Name=pelvis
  swing1LimitDegrees=45
  swing1Motion=Free
  swing2LimitDegrees=45
  swing2Motion=Free
  twistLimitDegrees=45
  twistMotion=Free
 }
]
