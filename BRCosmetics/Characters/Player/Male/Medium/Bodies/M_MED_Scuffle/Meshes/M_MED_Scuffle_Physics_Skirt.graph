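# Physics asset for the skirt on the Scuffle male medium player body, with back, left and right panels plus two side knives.
# Capsules on the pelvis, lower spine and thighs form the body volumes, and boxes shape the panels.
# The back panel chains from the pelvis and side panels from the thighs via tiny base bodies.
# Panel joints allow swing mostly about one axis, and knives hang off side panels with tight limits.
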PhysicsAsset /BRCosmetics/Characters/Player/Male/Medium/Bodies/M_MED_Scuffle/Meshes/M_MED_Scuffle_Physics_Skirt
bodies {
 dyn_skirt_bk_mid_01 [
  {shapeType=Capsule, length=0, radius=0, shapeName=dyn_skirt_bk_mid_01_capsule, center=<5.132 1.7853 -0.8882>, rotation=<rot 0.1051 -90 3.405>}
  {shapeType=Box, extentX=13.6052, extentY=2.351, extentZ=33.2367, shapeName=dyn_skirt_bk_mid_01_box, center=<7.8412 0 0>}
  {
   shapeType=Box
   extentX=14.324
   extentY=4.1993
   extentZ=13.3235
   shapeName=dyn_skirt_bk_mid_01_box
   center=<6.1102 4.5587 16.3122>
   rotation=<rot 0 0.0001 62.9746>
  }
  {
   shapeType=Box
   extentX=14.96
   extentY=4.5074
   extentZ=14.66
   shapeName=dyn_skirt_bk_mid_01_box
   center=<6.4429 7.6326 -18.5244>
   rotation=<rot 0 0 -60.519>
  }
 ]
 dyn_skirt_bk_mid_base [{shapeType=Capsule, length=0.505, radius=0.505, shapeName=dyn_skirt_bk_mid_base_capsule, rotation=<rot 0 -90 0>}]
 dyn_skirt_l_01 [
  {shapeType=Capsule, length=0, radius=0, shapeName=dyn_skirt_l_01_capsule, center=<7.2323 0.1786 -1.8823>, rotation=<rot 0 -72.9555 -88.2064>}
  {shapeType=Box, extentX=24.146, extentY=16.6444, extentZ=5.498, shapeName=dyn_skirt_l_01_box, center=<7.4048 0 0>}
 ]
 dyn_skirt_l_base [{shapeType=Capsule, length=0.505, radius=0.505, shapeName=dyn_skirt_l_base_capsule, rotation=<rot 0 -90 0>}]
 dyn_skirt_l_knife_base [{shapeType=Capsule, length=8, radius=2, shapeName=dyn_skirt_l_knife_base_capsule, center=<7.0584 0 0>, rotation=<rot 90 -179.9999 179.9999>}]
 dyn_skirt_r_01 [
  {shapeType=Capsule, length=0, radius=0, shapeName=dyn_skirt_l_01_capsule, center=<-7.2323 -0.1786 1.8823>, rotation=<rot 0 -72.9555 91.7936>}
  {shapeType=Box, extentX=24.146, extentY=16.5725, extentZ=5.7238, shapeName=dyn_skirt_l_01_box, center=<7.493 0 0>, rotation=<rot 0 0 -180>}
 ]
 dyn_skirt_r_base [{shapeType=Capsule, length=0.505, radius=0.505, shapeName=dyn_skirt_r_base_capsule, rotation=<rot 0 -90 0>}]
 dyn_skirt_r_knife_base [{shapeType=Capsule, length=8, radius=2, shapeName=dyn_skirt_r_knife_base_capsule, center=<7.0584 0 0>, rotation=<rot 90 -160.5288 -160.5286>}]
 pelvis [{shapeType=Capsule, length=11.4036, radius=10.3226, shapeName=pelvis_capsule, center=<-6.6284 0.2079 0>}]
 spine_01 [{shapeType=Capsule, length=14.2466, radius=8.1172, shapeName=spine_01_capsule, center=<-5.2886 -0.552 0>}]
 spine_02 [{shapeType=Capsule, length=14.5816, radius=7.7821, shapeName=spine_02_capsule, center=<-2.9832 0 0>}]
 thigh_l [{shapeType=Capsule, length=40.6898, radius=7.5867, shapeName=thigh_l_capsule, center=<-19.6494 0 0>, rotation=<rot 90 0 0>}]
 thigh_r [{shapeType=Capsule, length=40.4717, radius=7.6999, shapeName=thigh_r_capsule, center=<19.6104 0.0592 0>, rotation=<rot 90 0 0>}]
}
constraints [
 {
  bone1Name=spine_01
  bone2Name=pelvis
  swing1LimitDegrees=45
  swing1Motion=Free
  swing2LimitDegrees=45
  swing2Motion=Free
  twistLimitDegrees=45
  twistMotion=Free
 }
 {
  bone1Name=spine_02
  bone2Name=spine_01
  swing1LimitDegrees=45
  swing1Motion=Free
  swing2LimitDegrees=45
  swing2Motion=Free
  twistLimitDegrees=45
  twistMotion=Free
 }
 {
  bone1Name=thigh_l
  bone2Name=pelvis
  swing1LimitDegrees=45
  swing1Motion=Free
  swing2LimitDegrees=45
  swing2Motion=Free
  twistLimitDegrees=45
  twistMotion=Free
 }
 {
  bone1Name=thigh_r
  bone2Name=pelvis
  swing1LimitDegrees=45
  swing1Motion=Free
  swing2LimitDegrees=45
  swing2Motion=Free
  twistLimitDegrees=45
  twistMotion=Free
 }
 {
  bone1Name=dyn_skirt_bk_mid_base
  bone2Name=pelvis
  swing1LimitDegrees=45
  swing1Motion=Free
  swing2LimitDegrees=45
  swing2Motion=Free
  twistLimitDegrees=45
  twistMotion=Free
 }
 {bone1Name=dyn_skirt_bk_mid_01, bone2Name=dyn_skirt_bk_mid_base, swing1LimitDegrees=20, swing2LimitDegrees=0, twistLimitDegrees=0}
 {
  bone1Name=dyn_skirt_l_base
  bone2Name=thigh_l
  swing1LimitDegrees=45
  swing1Motion=Free
  swing2LimitDegrees=45
  swing2Motion=Free
  twistLimitDegrees=45
  twistMotion=Free
 }
 {bone1Name=dyn_skirt_l_01, bone2Name=dyn_skirt_l_base, swing1LimitDegrees=0, swing2LimitDegrees=30, twistLimitDegrees=2}
 {
  bone1Name=dyn_skirt_r_base
  bone2Name=thigh_r
  swing1LimitDegrees=45
  swing1Motion=Free
  swing2LimitDegrees=45
  swing2Motion=Free
  twistLimitDegrees=45
  twistMotion=Free
 }
 {bone1Name=dyn_skirt_r_01, bone2Name=dyn_skirt_r_base, swing1LimitDegrees=2, swing2LimitDegrees=30, twistLimitDegrees=2}
 {bone1Name=dyn_skirt_l_knife_base, bone2Name=dyn_skirt_l_01, swing1LimitDegrees=3, swing2LimitDegrees=3, twistLimitDegrees=0}
 {bone1Name=dyn_skirt_r_knife_base, bone2Name=dyn_skirt_r_01, swing1LimitDegrees=3, swing2LimitDegrees=3, twistLimitDegrees=0}
]
