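# Physics asset for the M_MED_Scuffle face accessory (Frontend variant by name), set up for its dangling left and right hood straps.
# Each strap is a three-capsule chain with limited swing and twist, hung from a zero-size base body constrained to neck_02.
# Sized capsules on the upper spine, clavicles and arms form the body volumes around the straps, linked by angularly free constraints through zero-size neck and spine_02 bodies.
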PhysicsAsset /BRCosmetics/Characters/Player/Male/Medium/Bodies/M_MED_Scuffle/Meshes/Parts/M_MED_Scuffle_FaceAcc_Physics_Frontend
bodies {
 clavicle_l [{shapeType=Capsule, length=9.9697, radius=6.4604, shapeName=clavicle_l_capsule, center=<7.5549 0 3.6499>, rotation=<rot 90 0 0>}]
 clavicle_r [{shapeType=Capsule, length=9.9697, radius=6.4604, shapeName=clavicle_r_capsule, center=<-5.2931 0 -3.548>, rotation=<rot 90 0 0>}]
 dyn_hood_strap_l_01 [{shapeType=Capsule, length=6, radius=2, shapeName=dyn_hood_strap_l_01_capsule, center=<4.7177 0 0>, rotation=<rot 90 0 0>}]
 dyn_hood_strap_l_02 [{shapeType=Capsule, length=11.5996, radius=1.5452, shapeName=dyn_hood_strap_l_02_capsule, center=<4.7177 0 0>, rotation=<rot 90 0 0>}]
 dyn_hood_strap_l_03 [{shapeType=Capsule, length=10.2017, radius=1.4364, shapeName=dyn_hood_strap_l_03_capsule, center=<5.9226 0 0>, rotation=<rot 90 0 0>}]
 dyn_hood_strap_l_base [{shapeType=Capsule, length=0, radius=0, shapeName=dyn_hood_strap_l_base_capsule}]
 dyn_hood_strap_r_01 [{shapeType=Capsule, length=8.7314, radius=1.98, shapeName=dyn_hood_strap_r_01_capsule, center=<4.7177 0 0>, rotation=<rot 90 0 0>}]
 dyn_hood_strap_r_02 [{shapeType=Capsule, length=10.2036, radius=1.4139, shapeName=dyn_hood_strap_r_02_capsule, center=<4.7177 0 0>, rotation=<rot 90 0 0>}]
 dyn_hood_strap_r_03 [{shapeType=Capsule, length=10.1874, radius=1.4506, shapeName=dyn_hood_strap_r_03_capsule, center=<5.9226 0 0>, rotation=<rot 90 0 0>}]
 dyn_hood_strap_r_base [{shapeType=Capsule, length=0, radius=0, shapeName=dyn_hood_strap_r_base_capsule}]
 lowerarm_l [{shapeType=Capsule, length=31.4435, radius=6, shapeName=lowerarm_l_capsule, center=<13.4694 0 0>, rotation=<rot 90 0 0>}]
 lowerarm_r [{shapeType=Capsule, length=27.6831, radius=6, shapeName=lowerarm_r_capsule, center=<-14.2651 0 0>, rotation=<rot 90 0 0>}]
 neck_01 [{shapeType=Capsule, length=0, radius=0, shapeName=neck_01_capsule}]
 neck_02 [{shapeType=Capsule, length=0, radius=0, shapeName=neck_02_capsule}]
 spine_02 [{shapeType=Capsule, length=0, radius=0, shapeName=spine_02_capsule}]
 spine_03 [{shapeType=Capsule, length=15.3516, radius=16.1252, shapeName=spine_03_capsule, center=<-0.3385 -2.0681 0>}]
 spine_04 [{shapeType=Capsule, length=17.0365, radius=16.2554, shapeName=spine_04_capsule, center=<2.2517 -2.6331 0>}]
 spine_05 [{shapeType=Capsule, length=4.6865, radius=16.3429, shapeName=spine_05_capsule, center=<-0.0436 -3.5427 0>}]
 upperarm_l [{shapeType=Capsule, length=30.2624, radius=5.5405, shapeName=upperarm_l_capsule, center=<12.7692 0 0>, rotation=<rot 0 -90 90>}]
 upperarm_r [{shapeType=Capsule, length=30.2624, radius=5.5405, shapeName=upperarm_r_capsule, center=<-12.4161 0 0>, rotation=<rot 0 -90 90>}]
}
constraints [
 {
  bone1Name=spine_03
  bone2Name=spine_02
  swing1LimitDegrees=45
  swing1Motion=Free
  swing2LimitDegrees=45
  swing2Motion=Free
  twistLimitDegrees=45
  twistMotion=Free
 }
 {
  bone1Name=spine_04
  bone2Name=spine_03
  swing1LimitDegrees=45
  swing1Motion=Free
  swing2LimitDegrees=45
  swing2Motion=Free
  twistLimitDegrees=45
  twistMotion=Free
 }
 {
  bone1Name=spine_05
  bone2Name=spine_04
  swing1LimitDegrees=45
  swing1Motion=Free
  swing2LimitDegrees=45
  swing2Motion=Free
  twistLimitDegrees=45
  twistMotion=Free
 }
 {
  bone1Name=clavicle_l
  bone2Name=spine_05
  swing1LimitDegrees=45
  swing1Motion=Free
  swing2LimitDegrees=45
  swing2Motion=Free
  twistLimitDegrees=45
  twistMotion=Free
 }
 {
  bone1Name=clavicle_r
  bone2Name=spine_05
  swing1LimitDegrees=45
  swing1Motion=Free
  swing2LimitDegrees=45
  swing2Motion=Free
  twistLimitDegrees=45
  twistMotion=Free
 }
 {
  bone1Name=neck_01
  bone2Name=spine_05
  swing1LimitDegrees=45
  swing1Motion=Free
  swing2LimitDegrees=45
  swing2Motion=Free
  twistLimitDegrees=45
  twistMotion=Free
 }
 {
  bone1Name=neck_02
  bone2Name=neck_01
  swing1LimitDegrees=45
  swing1Motion=Free
  swing2LimitDegrees=45
  swing2Motion=Free
  twistLimitDegrees=45
  twistMotion=Free
 }
 {
  bone1Name=dyn_hood_strap_l_base
  bone2Name=neck_02
  swing1LimitDegrees=45
  swing1Motion=Free
  swing2LimitDegrees=45
  swing2Motion=Free
  twistLimitDegrees=45
  twistMotion=Free
 }
 {bone1Name=dyn_hood_strap_r_03, bone2Name=dyn_hood_strap_r_02, swing1LimitDegrees=30, swing2LimitDegrees=30, twistLimitDegrees=15}
 {bone1Name=dyn_hood_strap_r_02, bone2Name=dyn_hood_strap_r_01, swing1LimitDegrees=30, swing2LimitDegrees=30, twistLimitDegrees=15}
 {bone1Name=dyn_hood_strap_r_01, bone2Name=dyn_hood_strap_r_base, swing1LimitDegrees=30, swing2LimitDegrees=30, twistLimitDegrees=15}
 {
  bone1Name=dyn_hood_strap_r_base
  bone2Name=neck_02
  swing1LimitDegrees=45
  swing1Motion=Free
  swing2LimitDegrees=45
  swing2Motion=Free
  twistLimitDegrees=45
  twistMotion=Free
 }
 {bone1Name=dyn_hood_strap_l_03, bone2Name=dyn_hood_strap_l_02, swing1LimitDegrees=20, swing2LimitDegrees=30, twistLimitDegrees=15}
 {bone1Name=dyn_hood_strap_l_02, bone2Name=dyn_hood_strap_l_01, swing1LimitDegrees=30, swing2LimitDegrees=10, twistLimitDegrees=15}
 {bone1Name=dyn_hood_strap_l_01, bone2Name=dyn_hood_strap_l_base, swing1LimitDegrees=30, swing2LimitDegrees=30, twistLimitDegrees=15}
 {
  bone1Name=upperarm_l
  bone2Name=clavicle_l
  swing1LimitDegrees=45
  swing1Motion=Free
  swing2LimitDegrees=45
  swing2Motion=Free
  twistLimitDegrees=45
  twistMotion=Free
 }
 {
  bone1Name=upperarm_r
  bone2Name=clavicle_r
  swing1LimitDegrees=45
  swing1Motion=Free
  swing2LimitDegrees=45
  swing2Motion=Free
  twistLimitDegrees=45
  twistMotion=Free
 }
 {
  bone1Name=lowerarm_l
  bone2Name=upperarm_l
  swing1LimitDegrees=45
  swing1Motion=Free
  swing2LimitDegrees=45
  swing2Motion=Free
  twistLimitDegrees=45
  twistMotion=Free
 }
 {
  bone1Name=lowerarm_r
  bone2Name=upperarm_r
  swing1LimitDegrees=45
  swing1Motion=Free
  swing2LimitDegrees=45
  swing2Motion=Free
  twistLimitDegrees=45
  twistMotion=Free
 }
]
