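# Animation Blueprint for the face accessory of the M_MED_Scuffle character, adding physics motion to its bones.
# It copies the attached parent's pose, layers in the Body part's clavicle branches, and drives the hood strap master with a weighted multi-bone constraint.
# That cached pose feeds two head-space RigidBody simulations, gameplay and frontend, blended by an idle-frontend flag, then slightly offsets the head.
# The event graph updates gravity override, frontend and LOD state.
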AnimBlueprint /BRCosmetics/Characters/Player/Male/Medium/Bodies/M_MED_Scuffle/Meshes/Parts/M_MED_Scuffle_FaceAcc_AnimBP
target_skeleton=&/M_MED_Scuffle_FaceAcc_Skeleton
graphs [
 AnimGraph {nodes [
   n1 = RigidBody!(ComponentPose=Local To Component(LocalPose=Use cached pose 'ToPhysics'()), Alpha, OverrideWorldGravity:Vector="0,0,0", ExternalForce) -> (Pose) {
    bClampLinearTranslationLimitToRefPose=true
    bOverrideWorldGravity=true
    overridePhysicsAsset=&/M_MED_Scuffle_FaceAcc_Physics
    simulationSpace=BaseBoneSpace
    alphaBoolBlend {blendOption=Linear}
    baseBoneRef {boneName=head}
    componentAppliedLinearAccClamp=<3000 3000 3000>
    simSpaceSettings {
     maxAngularAcceleration=1
     maxAngularVelocity=1
     maxLinearAcceleration=700
     maxLinearVelocity=500
     velocityScaleZ=0.1
     worldAlpha=1
     externalLinearDragV=<0.1 0.1>
    }
   }
   n2 = Component To Local(ComponentPose=@n1)
   n3 = RigidBody(ComponentPose=Local To Component(LocalPose=Use cached pose 'ToPhysics'()), OverrideWorldGravity="0,0,0") {
    bClampLinearTranslationLimitToRefPose=true
    bOverrideWorldGravity=true
    overridePhysicsAsset=&/M_MED_Scuffle_FaceAcc_Physics_Frontend
    simulationSpace=BaseBoneSpace
    alphaBoolBlend {blendOption=Linear}
    baseBoneRef {boneName=head}
    componentAppliedLinearAccClamp=<3000 3000 3000>
    simSpaceSettings {
     maxAngularAcceleration=40
     maxAngularVelocity=3
     maxLinearAcceleration=700
     maxLinearVelocity=500
     velocityScaleZ=0.1
     worldAlpha=1
     externalLinearDragV=<0.1 0.1>
    }
   }
   n4 = Component To Local(ComponentPose=@n3)
   n5 = Get Is Idle Frontend() -> (Is Idle Frontend:bool)
   n6 = "Transform (Modify) Bone - Bone: head"!(
    ComponentPose=Local To Component(LocalPose=Blend Poses by bool(BlendPose_0=@n4, BlendPose_1=@n2, bActiveValue=@n5) {blendType=Linear})
    bAlphaBoolEnabled:bool=True
    Translation="0,0.05,0"
    Rotation="0,0,0"
    Scale="1,1,1"
   ) -> (Pose) {
    alphaInputType=Bool
    rotationMode=BMM_Additive
    scaleMode=BMM_Additive
    translationMode=BMM_Additive
    alphaBoolBlend {blendOption=Linear}
    boneToModify {boneName=head}
    translation {y=0.05}
   }
   Output Pose(Result=Component To Local(ComponentPose=@n6)) {layerGroup=None}
   n7 = Copy Pose From Mesh(SourceMeshComponent=Get Body() -> (Body:Skeletal Mesh Component Budgeted Object Reference)) {bUseAttachedParent=true}
   n8 = Copy Pose From Mesh() {bUseAttachedParent=true}
   n9 = "Constraint - dyn_hood_strap_master (PPPPPPPP)"!(
    ComponentPose=Local To Component(LocalPose=Layered blend per bone(BasePose=@n8, BlendPoses_0=@n7) {layerSetup [{branchFilters [{blendDepth=3, boneName=clavicle_l}, {blendDepth=3, boneName=clavicle_r}]}]})
    Alpha=1
    ConstraintWeights_0=0.0355
    ConstraintWeights_1:float=0.8294
    ConstraintWeights_2:float=0.0153
    ConstraintWeights_3:float=0.0067
    ConstraintWeights_4:float=0.0964
    ConstraintWeights_5:float=0.0008
    ConstraintWeights_6:float=0.0056
    ConstraintWeights_7:float=0.0102
   ) -> (Pose) {
    alphaBoolBlend {blendOption=Linear}
    boneToModify {boneName=dyn_hood_strap_master}
    constraintSetup [
     {targetBone {boneName=spine_04}}
     {targetBone {boneName=head}}
     {targetBone {boneName=neck_02}}
     {targetBone {boneName=neck_01}}
     {targetBone {boneName=spine_05}}
     {targetBone {boneName=clavicle_l}}
     {targetBone {boneName=clavicle_r}}
     {targetBone {boneName=spine_03}}
    ]
    constraintWeights [1,1,1,1,1,1,1,1]
   }
   Save cached pose 'ToPhysics'(Pose=Component To Local(ComponentPose=@n9)) {cacheName=ToPhysics}
  ]}
 EventGraph {nodes [
   Event BlueprintUpdateAnimation()
   > n1 = Gravity Override State Function DEPRECATED(
    CharacterPartAnimInstance=Self-Reference()
    GameTimeInSeconds=GetGameTimeInSeconds()
    GravityOverrideParams=Get Gravity Override Params()
    DebugJointName=dyn_hackBottomArmor
   )
   > Set Override World Gravity(Override World Gravity:Vector=@n1) -> (Output_Get:Vector)
   n2 = Get bIsPlayingEmote()
   > Set Is Idle Frontend(Is Idle Frontend:bool=AND Boolean(A=NOT Boolean(A=@n2), B=Get bIsInFrontEnd())) -> (Output_Get:bool)
   > Set Is LOD1(Is LOD1:bool="integer > integer"(A=Get CurrentLOD())) -> (Output_Get:bool)
   n3 = Event OnCharacterCustomizationLoaded()
   > Set Body(Body:Skeletal Mesh Component Budgeted Object Reference=Get Array Item with EFort Custom Part Type(Array=@n3.CustomMeshParts)) -> (Output_Get:Skeletal Mesh Component Budgeted Object Reference)
  ]}
]
variables [
 Body {type=SkeletalMeshComponentBudgeted}
 Gravity Override Params {type=GravityOverrideParamsStruct}
 Is Idle Frontend {type=boolean}
 Is LOD1 {type=boolean}
 Override World Gravity {type=Vector}
]
